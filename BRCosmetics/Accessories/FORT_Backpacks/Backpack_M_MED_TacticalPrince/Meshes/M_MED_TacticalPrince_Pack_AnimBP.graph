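# Animation Blueprint for the TacticalPrince backpack cosmetic that adjusts the pack bone on the wearing character.
# The AnimGraph chains additive Modify Bone nodes in two cached-pose stages, with some offsets gated by bool or float alphas.
# Each update derives pack rotation and offset variables from spine bend, plus skydive-dive and strafe flags, which clear while targeting.
# Customization load records wearer gender, body type and whether the body mesh matches TacticalPrince.
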AnimBlueprint /BRCosmetics/Accessories/FORT_Backpacks/Backpack_M_MED_TacticalPrince/Meshes/M_MED_TacticalPrince_Pack_AnimBP
target_skeleton=&/M_MED_TacticalPrince_Pack_Skeleton
graphs [
 AnimGraph {nodes [
   Output Pose(Result=Use cached pose 'CR_Cache_Output'()) {layerGroup=None}
   n1 = "Transform (Modify) Bone - Bone: pack"!(ComponentPose=Local To Component(LocalPose=Input Pose()), bAlphaBoolEnabled:bool, Translation="0,-1.5,0", Rotation="0,0,0", Scale) -> (Pose) {
    alphaInputType=Bool
    rotationMode=BMM_Additive
    rotationSpace=BCS_BoneSpace
    scaleMode=BMM_Additive
    scaleSpace=BCS_BoneSpace
    translationMode=BMM_Additive
    translationSpace=BCS_BoneSpace
    alphaBoolBlend {blendOption=Linear}
    alphaScaleBiasClamp {interpSpeedDecreasing=5, interpSpeedIncreasing=5}
    boneToModify {boneName=pack}
    translation {y=-1.5}
   }
   n2 = "Transform (Modify) Bone - Bone: pack"(ComponentPose=@n1, Translation="0,0.6,0", Rotation="0,0,0") {
    alphaInputType=Bool
    rotationMode=BMM_Additive
    rotationSpace=BCS_BoneSpace
    scaleMode=BMM_Additive
    scaleSpace=BCS_BoneSpace
    translationMode=BMM_Additive
    translationSpace=BCS_BoneSpace
    alphaBoolBlend {blendOption=Linear}
    alphaScaleBiasClamp {interpSpeedDecreasing=5, interpSpeedIncreasing=5}
    boneToModify {boneName=pack}
    translation {y=0.6}
   }
   Save cached pose 'CR_Cache_Input'(Pose=Component To Local(ComponentPose=@n2)) {cacheName=CR_Cache_Input}
   n3 = "Transform (Modify) Bone - Bone: pack"!(ComponentPose=Local To Component(LocalPose=Use cached pose 'CR_Cache_Input'()), Alpha:float, Translation="0,0,0", Rotation="0,0,0", Scale) -> (Pose) {
    rotationMode=BMM_Additive
    scaleMode=BMM_Additive
    scaleSpace=BCS_BoneSpace
    translationMode=BMM_Additive
    alphaBoolBlend {blendOption=Linear}
    boneToModify {boneName=pack}
   }
   n4 = "Transform (Modify) Bone - Bone: pack"(ComponentPose=@n3, Translation="0,0,0", Rotation="10,0,5") {
    rotationMode=BMM_Additive
    scaleMode=BMM_Additive
    scaleSpace=BCS_BoneSpace
    translationMode=BMM_Additive
    alphaBoolBlend {blendOption=Linear}
    alphaScaleBiasClamp {bInterpResult=true, interpSpeedDecreasing=5, interpSpeedIncreasing=5}
    boneToModify {boneName=pack}
    rotation {pitch=10, roll=5}
   }
   n5 = "Transform (Modify) Bone - Bone: pack"!(ComponentPose=@n4, bAlphaBoolEnabled:bool, Translation="1,-0.5,3", Rotation="0,10,10") -> (Pose) {
    alphaInputType=Bool
    rotationMode=BMM_Additive
    translationMode=BMM_Additive
    alphaBoolBlend {blendInTime=0.2, blendOption=Linear, blendOutTime=0.2}
    boneToModify {boneName=pack}
    rotation {roll=10, yaw=10}
    translation=<1 -0.5 3>
   }
   n6 = "Transform (Modify) Bone - Bone: pack"(ComponentPose=@n5, Translation="0,-0.25,0", Rotation="0,0,5") {
    alphaInputType=Bool
    rotationMode=BMM_Additive
    translationMode=BMM_Additive
    alphaBoolBlend {blendInTime=0.2, blendOption=Linear, blendOutTime=0.2}
    boneToModify {boneName=pack}
    rotation {roll=5}
    translation {y=-0.25}
   }
   Save cached pose 'CR_Cache_Output'(Pose=Component To Local(ComponentPose=@n6)) {cacheName=CR_Cache_Output}
  ]}
 EventGraph {nodes [
   Event BlueprintUpdateAnimation()
   > n1 = Sequence()
   n2 = GetOwnerSkeletalMeshComponent()
   n3 = BreakRotator(InRot=BreakTransform(InTransform=GetDeltaTransformFromRefPose(self=@n2, BoneName=spine_05, BaseName=spine_01)).Rotation)
   > Set backpack_rotate(backpack_rotate=MakeRotator(Roll=MapRangeClamped(Value=@n3.Yaw, InRangeA=0, InRangeB=-30, OutRangeA=5, OutRangeB=25)))
   > Set backpack_translation(backpack_translation=MakeVector(Y=MapRangeClamped(Value=@n3.Yaw, InRangeA=0, InRangeB=-30, OutRangeB=-6.5)))
   > n4 = Branch(Condition=Get bIsSkydiveDiveMode())
   > Set IsFloatingMode(IsFloatingMode:double=1) -> (Output_Get:double)
   Set IsFloatingMode(execute=@n4.else, IsFloatingMode=0)
   n5 = PlayerMovement(execute=@n1.then_1, AnimInstance=Self-Reference(), XDotProductCoefficient=1, YDotProductCoefficient=1, ZDotProductCoefficient=1)
   > n6 = Branch(Condition=Get bIsTargeting())
   > Set strafeLeft int(strafeLeft int:bool=false) -> (Output_Get:bool)
   > Set strafeRight int(strafeRight int:bool=false) -> (Output_Get:bool)
   > Set strafeForward int(strafeForward int:bool=false) -> (Output_Get:bool)
   > Set strafeBackward int(strafeBackward int:bool=false) -> (Output_Get:bool)
   n7 = BreakVector(InVec=MakeVector(X=@n5, Y=@n5.YMovement))
   n8 = "Clamp (Float)"(Value=@n7, Min=-1, Max=0)
   Set strafeLeft int(execute=@n6.else, strafeLeft int="To Bool (Double)"!(Value:double="Absolute (Float)"(A=@n8)) -> (ReturnValue:bool))
   > Set strafeRight int(strafeRight int="To Bool (Double)"(Value="Clamp (Float)"(Value=@n7, Min=0, Max=1)))
   > Set strafeForward int(strafeForward int="To Bool (Double)"(Value="Clamp (Float)"(Value=@n7.Y, Min=0, Max=1)))
   n9 = "Clamp (Float)"(Value=@n7.Y, Min=-1, Max=0)
   > Set strafeBackward int(strafeBackward int="To Bool (Double)"(Value="Absolute (Float)"(A=@n9)))
   Event OnCharacterCustomizationLoaded()
   > n10 = Set Owning Pawn(Owning Pawn:Pawn Object Reference=TryGetFortPawn()) -> (Output_Get:Pawn Object Reference)
   > Is Valid(InputObject=@n10)
   > n11 = Sequence!(execute) -> (then_0, then_1, then_2:Exec)
   n12 = FindGender!(execute=@n11.then_1, Object:Object Reference=Get Owning Pawn() -> (Owning Pawn:Pawn Object Reference)) -> (True:Exec, False:Exec, OutGender:EFortCustomGender Enum=Invalid)
   > Set IsFemale(IsFemale:bool="Equal (Enum)"(A:EFortCustomGender Enum=@n12, B:EFortCustomGender Enum=Female)) -> (Output_Get:bool)
   n13 = FindBodyType!(execute=@n11.then_2, Object:Object Reference=Get Owning Pawn()) -> (True:Exec, False:Exec, OutBodyType:EFortCustomBodyType Enum=NONE)
   > Set isLarge(isLarge:bool="Equal (Enum)"(A:EFortCustomBodyType Enum=@n13, B:EFortCustomBodyType Enum=Large)) -> (Output_Get:bool)
   > Set isMedium(isMedium:bool="Equal (Enum)"(A=@n13, B=Medium)) -> (Output_Get:bool)
   n14 = Cast To FortPlayerPawn(execute=@n11.then_0, Object=Get Owning Pawn())
   > n15 = Set Owning Player Pawn(Owning Player Pawn:Fort Player Pawn Object Reference=@n14) -> (Output_Get:Fort Player Pawn Object Reference)
   > Set Body Mesh(Body Mesh:Skeletal Mesh Component Object Reference=GetSkeletalMeshForPartType(self=@n15)) -> (Output_Get:Skeletal Mesh Component Object Reference)
   n16 = Get Body Mesh() -> (Body Mesh:Skeletal Mesh Component Object Reference)
   > Set Character(Character:bool=MatchesWildcard(SourceString=GetDisplayName(Object=@n16), Wildcard=*M_MED_TacticalPrince*, SearchCase=CaseSensitive)) -> (Output_Get:bool)
   Event BlueprintInitializeAnimation()
   > Set Skel_Mesh(Skel_Mesh=GetOwnerSkeletalMeshComponent())
   TryGetPawnOwner()
  ]}
]
variables [
 backpack_rotate {type=Rotator}
 backpack_translation {type=Vector}
 Body Mesh {type=SkeletalMeshComponent}
 Character {type=boolean}
 IsFemale {type=boolean}
 IsFloatingMode {type=number}
 isLarge {type=boolean}
 isMedium {type=boolean}
 moveBackward {type=number}
 moveForward {type=number}
 Owning Pawn {type=Pawn, default=null}
 Owning Player Pawn {type=FortPlayerPawn}
 Skel_Mesh {type=SkeletalMeshComponent}
 strafeBackward int {type=boolean}
 strafeForward int {type=boolean}
 strafeLeft float {type=number}
 strafeLeft int {type=boolean}
 strafeRight float {type=number}
 strafeRight int {type=boolean}
]
